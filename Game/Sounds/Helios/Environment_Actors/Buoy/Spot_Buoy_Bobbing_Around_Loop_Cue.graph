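# Ambient sound cue for a bobbing buoy, combining a continuous bobbing loop with intermittent metal pings.
# A Mixer blends a random pick of two bobbing branches, one pitched lower, with a looped random choice of six pings after random delays.
# Looping envelopers slowly vary each bobbing branch's volume and pitch.
# Separate attenuation overrides keep the bobbing local with listener focus while the pings carry much farther on a custom falloff curve.
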SoundCue /Game/Sounds/Helios/Environment_Actors/Buoy/Spot_Buoy_Bobbing_Around_Loop_Cue
root=Mixer_0
nodes [
 {
  id=Attenuation_0
  children [Looping_0]
  properties {
   bOverrideAttenuation=true
   attenuationOverrides {
    bEnableListenerFocus=true
    distanceAlgorithm=Custom
    falloffDistance=10000
    nonFocusVolumeAttenuation=0.75
    stereoSpread=50
    attenuationShapeExtents {x=0}
    customAttenuationCurve {editorCurveData {keys [
       <at 0 1>
       <at 0.2 1>
       {arriveTangent=-0.9113, interpMode=RCIM_Cubic, leaveTangent=-0.9113, tangentMode=RCTM_User, time=1, value=0.5}
       {arriveTangent=-0.016, interpMode=RCIM_Cubic, leaveTangent=-0.1171, tangentMode=RCTM_User, time=1.2, value=0}
      ]}}
   }
  }
 }
 {
  id=Attenuation_2
  children [Enveloper_3]
  properties {
   bOverrideAttenuation=true
   attenuationOverrides {
    attenuationShape=Capsule
    bEnableFocusInterpolation=true
    bEnableListenerFocus=true
    bEnablePriorityAttenuation=true
    falloffDistance=850
    focusAttackInterpSpeed=0.5
    focusAzimuth=15
    focusReleaseInterpSpeed=0.5
    focusVolumeAttenuation=1.2
    nonFocusAzimuth=45
    nonFocusVolumeAttenuation=0.5
    submixSendSettings [{
      bEnableHPFCutoff=false
      bEnableLPFCutoff=false
      disableManualSendClamp=false
      hPFCutoff=20
      lPFCutoff=20000
      maxSendDistance=6000
      maxSendLevel=1
      minSendDistance=400
      minSendLevel=0
      sendLevel=0.2
      sendLevelControlMethod=Linear
      soundSubmix=null
     }]
   }
  }
 }
 {
  id=Attenuation_3
  children [Enveloper_2]
  properties {
   bOverrideAttenuation=true
   attenuationOverrides {
    attenuationShape=Capsule
    bEnableFocusInterpolation=true
    bEnableListenerFocus=true
    bEnablePriorityAttenuation=true
    falloffDistance=850
    focusAttackInterpSpeed=0.5
    focusAzimuth=15
    focusReleaseInterpSpeed=0.5
    focusVolumeAttenuation=1.2
    nonFocusAzimuth=45
    nonFocusVolumeAttenuation=0.5
    submixSendSettings [{
      bEnableHPFCutoff=false
      bEnableLPFCutoff=false
      disableManualSendClamp=false
      hPFCutoff=20
      lPFCutoff=20000
      maxSendDistance=6000
      maxSendLevel=1
      minSendDistance=400
      minSendLevel=0
      sendLevel=0.2
      sendLevelControlMethod=Linear
      soundSubmix=null
     }]
   }
  }
 }
 {id=Delay_1, children [Random_1], properties {delayMax=18, delayMin=6}}
 {
  id=Enveloper_0
  children [Modulator_0]
  properties {
   bLoop=true
   bLoopIndefinitely=true
   loopEnd=30
   pitchCurve {editorCurveData {keys [
      {arriveTangent=-0.0013, interpMode=RCIM_Cubic, leaveTangent=-0.0013, time=0}
      <at 15 0.98 RCIM_Cubic>
      {arriveTangent=0.0013, interpMode=RCIM_Cubic, leaveTangent=0.0013, time=30}
     ]}}
   volumeCurve {editorCurveData {keys [<at 0 1>]}}
  }
 }
 {
  id=Enveloper_1
  children [Modulator_1]
  properties {
   bLoop=true
   bLoopIndefinitely=true
   loopEnd=30
   pitchCurve {editorCurveData {keys [
      {arriveTangent=-0.0013, interpMode=RCIM_Cubic, leaveTangent=-0.0013, time=0}
      <at 15 0.98 RCIM_Cubic>
      {arriveTangent=0.0013, interpMode=RCIM_Cubic, leaveTangent=0.0013, time=30}
     ]}}
   volumeCurve {editorCurveData {keys [<at 0 1>]}}
  }
 }
 {
  id=Enveloper_2
  children [Enveloper_1]
  properties {
   bLoop=true
   bLoopIndefinitely=true
   loopEnd=10
   pitchCurve {editorCurveData {keys [{time=0}]}}
   volumeCurve {editorCurveData {keys [
      {arriveTangent=-0.01, interpMode=RCIM_Cubic, time=0, value=1}
      <at 5 0.7 RCIM_Cubic>
      {arriveTangent=-0.03, interpMode=RCIM_Cubic, leaveTangent=-0.03, time=10, value=1}
      <at 15 0.4 RCIM_Cubic>
      <at 20 1 RCIM_Cubic>
     ]}}
  }
 }
 {
  id=Enveloper_3
  children [Enveloper_0]
  properties {
   bLoop=true
   bLoopIndefinitely=true
   loopEnd=10
   pitchCurve {editorCurveData {keys [{time=0}]}}
   volumeCurve {editorCurveData {keys [
      {arriveTangent=-0.01, interpMode=RCIM_Cubic, time=0, value=1}
      <at 5 0.7 RCIM_Cubic>
      {arriveTangent=-0.03, interpMode=RCIM_Cubic, leaveTangent=-0.03, time=10, value=1}
      <at 15 0.4 RCIM_Cubic>
      <at 20 1 RCIM_Cubic>
     ]}}
  }
 }
 {id=Looping_0, children [Delay_1]}
 {id=Mixer_0, children [Attenuation_0, Random_0], properties {inputVolume [1.2,1]}}
 {id=Modulator_0, children [WavePlayer_2], properties {pitchMax=0.9, pitchMin=0.8}}
 {id=Modulator_1, children [WavePlayer_3], properties {pitchMax=1.1, pitchMin=1}}
 {id=Random_0, children [Attenuation_2, Attenuation_3], properties {weights [1,1]}}
 {id=Random_1, children [WavePlayer_4,WavePlayer_5,WavePlayer_6,WavePlayer_7,WavePlayer_8,WavePlayer_9], properties {weights [1,1,1,1,1,1]}}
 {id=WavePlayer_2, properties {bLooping=true, soundWaveAssetPtr=&/Spot_Buoy_Bobbing_Around_Loop}}
 {id=WavePlayer_3, properties {bLooping=true, soundWaveAssetPtr=&Spot_Buoy_Bobbing_Around_Loop}}
 {id=WavePlayer_4, properties {soundWaveAssetPtr=&/Spot_Buoy_Metal_Ping_01}}
 {id=WavePlayer_5, properties {soundWaveAssetPtr=&/Spot_Buoy_Metal_Ping_02}}
 {id=WavePlayer_6, properties {soundWaveAssetPtr=&/Spot_Buoy_Metal_Ping_03}}
 {id=WavePlayer_7, properties {soundWaveAssetPtr=&/Spot_Buoy_Metal_Ping_04}}
 {id=WavePlayer_8, properties {soundWaveAssetPtr=&/Spot_Buoy_Metal_Ping_05}}
 {id=WavePlayer_9, properties {soundWaveAssetPtr=&/Spot_Buoy_Metal_Ping_06}}
]
properties {
 bOverrideConcurrency=true
 maxDistance=12000
 soundClassObject=&Fort_Ambient_Outdoors_Athena
 subtitlePriority=5000
 volumeMultiplier=0.5
 attenuationOverrides {
  attenuationShape=Capsule
  bEnableFocusInterpolation=true
  bEnableListenerFocus=true
  dBAttenuationAtMax=-40
  falloffDistance=750
  focusAzimuth=15
  focusVolumeAttenuation=1.2
  nonFocusAzimuth=45
  nonFocusVolumeAttenuation=0.5
  occlusionInterpolationTime=1
  occlusionLowPassFilterFrequency=10000
  occlusionTraceChannel=ECC_GameTraceChannel11
  occlusionVolumeAttenuation=0.5
  attenuationShapeExtents {x=590, z=90}
 }
 busSends [{audioBus=null, minSendLevel=0.6, sendLevel=1, soundSourceBus=&AmbientAudio_Exterior_SourceBus}]
 concurrencyOverrides {bLimitToOwner=true, maxCount=2}
}
