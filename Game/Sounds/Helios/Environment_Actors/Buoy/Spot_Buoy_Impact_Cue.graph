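# Sound cue for impact and damage hits on the Helios buoy environment actor.
# A random node picks one of three equally weighted damage waves, and a modulator adds slight pitch variation and a raised volume range.
# It plays in the impacts sound class with attenuation borrowed from a cedar forest chime and a concurrency set borrowed from a dojo gong.
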SoundCue /Game/Sounds/Helios/Environment_Actors/Buoy/Spot_Buoy_Impact_Cue
root=Modulator_0
nodes [
 {id=Modulator_0, children [Random_1], properties {pitchMax=1.02, pitchMin=0.98, volumeMax=1.5, volumeMin=1.3}}
 {id=Random_1, children [WavePlayer_0,WavePlayer_1,WavePlayer_2], properties {bShouldExcludeFromBranchCulling=true, weights [1,1,1]}}
 {id=WavePlayer_0, properties {soundWaveAssetPtr=&/Spot_Buoy_Impact_Damage01}}
 {id=WavePlayer_1, properties {soundWaveAssetPtr=&/Spot_Buoy_Impact_Damage02}}
 {id=WavePlayer_2, properties {soundWaveAssetPtr=&/Spot_Buoy_Impact_Damage03}}
]
properties {
 attenuationSettings=/Game/Sounds/Fort_Ambient_Looping/CedarForestChime/CedarWoods_Chime_Impact_Att
 maxDistance=4800
 soundClassObject=&Fort_Impacts
 volumeMultiplier=4
 concurrencyOverrides {bLimitToOwner=true, maxCount=1, resolutionRule=StopOldest, volumeScaleAttackTime=1}
 concurrencySet [/Game/Sounds/Asteria/Environment_Actors/Gong/Dojo_Gong_Conc]
}
